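# Animation Blueprint for the Spire boss, covering idle, death, hit reactions and beam attack posing.
# An idle/death state machine picks a full-body death reaction from the active slam appendage or beam sweep, then plays death.
# Layered on top are four event-triggered, self-clearing additive hit reacts for pillar and shards, then optional procedural shard yaw.
# A final beam layer adds firing recoil, bone-driven knockback and a pitch aim-offset blendspace.
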AnimBlueprint /SpireBoss/Assets/Animation/SpireBoss_AnimBP
target_skeleton=/SpireBoss/GameplayActors/Meshes/Spires_Creature_Skeleton
graphs [
 AdditiveHitReacts {nodes [
   n1 = Use cached pose 'Input Pose'()
   AnimGraphNode_StateMachine_5 = HitReact_Shield_Left() {is=title AnimGraphNode_StateMachine, maxTransitionsPerFrame=1, maxTransitionsRequests=1}
   AnimGraphNode_StateMachine_3 = HitReact_Shield_Right() {is=title AnimGraphNode_StateMachine, maxTransitionsPerFrame=1, maxTransitionsRequests=1}
   n2 = Apply Additive(
    Base=Apply Additive(Base=Slot 'FullbodyAdditive'(Source=@n1) {slotName=FullbodyAdditive}, Additive=@AnimGraphNode_StateMachine_5) {alphaBoolBlend {blendOption=Linear}}
    Additive=@AnimGraphNode_StateMachine_3
   ) {alphaBoolBlend {blendOption=Linear}}
   AnimGraphNode_StateMachine_6 = HitReact_Shield_Rear() {is=title AnimGraphNode_StateMachine, maxTransitionsPerFrame=1, maxTransitionsRequests=1}
   AnimGraphNode_StateMachine_1 = HitReact_MainPillar() {is=title AnimGraphNode_StateMachine, maxTransitionsPerFrame=1, maxTransitionsRequests=1}
   Output Pose(Result=Apply Additive(
     Base=Apply Additive(Base=@n2, Additive=@AnimGraphNode_StateMachine_6) {alphaBoolBlend {blendOption=Linear}}
     Additive=@AnimGraphNode_StateMachine_1
    ) {alphaBoolBlend {blendOption=Linear}})
   Save cached pose 'Input Pose'(Pose=Input Pose() {name=InPose_0}) {cacheName=Input Pose}
  ]}
 HitReact_MainPillar {
  entry=Passthrough
  path=AdditiveHitReacts.AnimGraphNode_StateMachine_1.HitReact_MainPillar
  ids=id
  nodes [
   To Hit React() {is=title AnimStateAliasNode, bGlobalAlias=true}
   AnimStateNode_0 = Passthrough() {is=title AnimStateNode}
   AnimStateNode_1 = HitReact_Main() {is=title AnimStateNode}
   AnimStateTransitionNode_1 = AnimStateTransitionNode() {bAutomaticRuleBasedOnSequencePlayerInState=true, crossfadeDuration=0}
   AnimStateTransitionNode_3 = AnimStateTransitionNode() {crossfadeDuration=0}
  ]
  links [
   {
    dst=Passthrough
    rule=AdditiveHitReacts.AnimGraphNode_StateMachine_1.HitReact_MainPillar.AnimStateTransitionNode_1.Transition
    src=HitReact_Main
    via=AnimStateTransitionNode_1
   }
   {
    dst=HitReact_Main
    rule=AdditiveHitReacts.AnimGraphNode_StateMachine_1.HitReact_MainPillar.AnimStateTransitionNode_3.Transition
    src=To Hit React
    via=AnimStateTransitionNode_3
   }
  ]
 }
 Passthrough {
  path=AdditiveHitReacts.AnimGraphNode_StateMachine_1.HitReact_MainPillar.AnimStateNode_0.Passthrough
  nodes [Output Animation Pose(Result=Additive Identity Pose())]
 }
 HitReact_Main {
  path=AdditiveHitReacts.AnimGraphNode_StateMachine_1.HitReact_MainPillar.AnimStateNode_1.HitReact_Main
  nodes [Output Animation Pose(Result="Sequence Player (additive) 'SpireBoss_Hitreact_Pillar'"() {bLoopAnimation=false, sequence=&/SpireBoss_Hitreact_Pillar}) {
    becomeRelevantFunction {bSelfContext=true, memberGuid=#1, memberName=OnMainBecomeRelevant}
    node {becomeRelevantFunction {functionName=OnBecomeRelevant}}
   }]
 }
 Transition {path=AdditiveHitReacts.AnimGraphNode_StateMachine_1.HitReact_MainPillar.AnimStateTransitionNode_1.Transition, nodes [Result()]}
 Transition {
  path=AdditiveHitReacts.AnimGraphNode_StateMachine_1.HitReact_MainPillar.AnimStateTransitionNode_3.Transition
  nodes [Result(bCanEnterTransition=Get HitReact_Spire_Main() -> (HitReact_Spire_Main:bool))]
 }
 HitReact_Shield_Right {
  entry=Passthrough
  path=AdditiveHitReacts.AnimGraphNode_StateMachine_3.HitReact_Shield_Right
  ids=id
  nodes [
   To Hit React() {bGlobalAlias=true}
   AnimStateNode_0 = Passthrough()
   AnimStateNode_1 = HitReact_Shard_Right() {is=title AnimStateNode}
   AnimStateTransitionNode_1 = AnimStateTransitionNode() {bAutomaticRuleBasedOnSequencePlayerInState=true, crossfadeDuration=0}
   AnimStateTransitionNode_2 = AnimStateTransitionNode() {crossfadeDuration=0}
  ]
  links [
   {
    dst=Passthrough
    rule=AdditiveHitReacts.AnimGraphNode_StateMachine_3.HitReact_Shield_Right.AnimStateTransitionNode_1.Transition
    src=HitReact_Shard_Right
    via=AnimStateTransitionNode_1
   }
   {
    dst=HitReact_Shard_Right
    rule=AdditiveHitReacts.AnimGraphNode_StateMachine_3.HitReact_Shield_Right.AnimStateTransitionNode_2.Transition
    src=To Hit React
    via=AnimStateTransitionNode_2
   }
  ]
 }
 Passthrough {
  path=AdditiveHitReacts.AnimGraphNode_StateMachine_3.HitReact_Shield_Right.AnimStateNode_0.Passthrough
  nodes [Output Animation Pose(Result=Additive Identity Pose())]
 }
 HitReact_Shard_Right {
  path=AdditiveHitReacts.AnimGraphNode_StateMachine_3.HitReact_Shield_Right.AnimStateNode_1.HitReact_Shard_Right
  nodes [Output Animation Pose(Result="Sequence Player (additive) 'SpireBoss_Hitreact_Shard_Right'"() {bLoopAnimation=false, sequence=&/SpireBoss_Hitreact_Shard_Right}) {
    becomeRelevantFunction {bSelfContext=true, memberGuid=#2, memberName=OnRightHitBecomeRelevant}
    node {becomeRelevantFunction {functionName=OnBecomeRelevant}}
   }]
 }
 Transition {path=AdditiveHitReacts.AnimGraphNode_StateMachine_3.HitReact_Shield_Right.AnimStateTransitionNode_1.Transition, nodes [Result()]}
 Transition {
  path=AdditiveHitReacts.AnimGraphNode_StateMachine_3.HitReact_Shield_Right.AnimStateTransitionNode_2.Transition
  nodes [Result(bCanEnterTransition=Get HitReact_Spire_Shard_Right() -> (HitReact_Spire_Shard_Right:bool))]
 }
 HitReact_Shield_Left {
  entry=Passthrough
  path=AdditiveHitReacts.AnimGraphNode_StateMachine_5.HitReact_Shield_Left
  ids=id
  nodes [
   To Hit React() {bGlobalAlias=true}
   AnimStateNode_0 = Passthrough()
   AnimStateNode_1 = HitReact_Shard_Left() {is=title AnimStateNode}
   AnimStateTransitionNode_1 = AnimStateTransitionNode() {bAutomaticRuleBasedOnSequencePlayerInState=true, crossfadeDuration=0}
   AnimStateTransitionNode_2 = AnimStateTransitionNode() {crossfadeDuration=0}
  ]
  links [
   {
    dst=Passthrough
    rule=AdditiveHitReacts.AnimGraphNode_StateMachine_5.HitReact_Shield_Left.AnimStateTransitionNode_1.Transition
    src=HitReact_Shard_Left
    via=AnimStateTransitionNode_1
   }
   {
    dst=HitReact_Shard_Left
    rule=AdditiveHitReacts.AnimGraphNode_StateMachine_5.HitReact_Shield_Left.AnimStateTransitionNode_2.Transition
    src=To Hit React
    via=AnimStateTransitionNode_2
   }
  ]
 }
 Passthrough {
  path=AdditiveHitReacts.AnimGraphNode_StateMachine_5.HitReact_Shield_Left.AnimStateNode_0.Passthrough
  nodes [Output Animation Pose(Result=Additive Identity Pose())]
 }
 HitReact_Shard_Left {
  path=AdditiveHitReacts.AnimGraphNode_StateMachine_5.HitReact_Shield_Left.AnimStateNode_1.HitReact_Shard_Left
  nodes [Output Animation Pose(Result="Sequence Player (additive) 'SpireBoss_Hitreact_Shard_Left'"() {bLoopAnimation=false, sequence=&/SpireBoss_Hitreact_Shard_Left}) {
    becomeRelevantFunction {bSelfContext=true, memberGuid=#3, memberName=OnLeftHitBecomeRelevant}
    node {becomeRelevantFunction {functionName=OnBecomeRelevant}}
   }]
 }
 Transition {path=AdditiveHitReacts.AnimGraphNode_StateMachine_5.HitReact_Shield_Left.AnimStateTransitionNode_1.Transition, nodes [Result()]}
 Transition {
  path=AdditiveHitReacts.AnimGraphNode_StateMachine_5.HitReact_Shield_Left.AnimStateTransitionNode_2.Transition
  nodes [Result(bCanEnterTransition=Get HitReact_Spire_Shard_Left() -> (HitReact_Spire_Shard_Left:bool))]
 }
 HitReact_Shield_Rear {
  entry=Passthrough
  path=AdditiveHitReacts.AnimGraphNode_StateMachine_6.HitReact_Shield_Rear
  ids=id
  nodes [
   To Hit React() {bGlobalAlias=true}
   AnimStateNode_0 = Passthrough()
   AnimStateNode_1 = HitReact_Shard_Rear() {is=title AnimStateNode}
   AnimStateTransitionNode_1 = AnimStateTransitionNode() {bAutomaticRuleBasedOnSequencePlayerInState=true, crossfadeDuration=0}
   AnimStateTransitionNode_2 = AnimStateTransitionNode() {crossfadeDuration=0}
  ]
  links [
   {
    dst=Passthrough
    rule=AdditiveHitReacts.AnimGraphNode_StateMachine_6.HitReact_Shield_Rear.AnimStateTransitionNode_1.Transition
    src=HitReact_Shard_Rear
    via=AnimStateTransitionNode_1
   }
   {
    dst=HitReact_Shard_Rear
    rule=AdditiveHitReacts.AnimGraphNode_StateMachine_6.HitReact_Shield_Rear.AnimStateTransitionNode_2.Transition
    src=To Hit React
    via=AnimStateTransitionNode_2
   }
  ]
 }
 Passthrough {
  path=AdditiveHitReacts.AnimGraphNode_StateMachine_6.HitReact_Shield_Rear.AnimStateNode_0.Passthrough
  nodes [Output Animation Pose(Result=Additive Identity Pose())]
 }
 HitReact_Shard_Rear {
  path=AdditiveHitReacts.AnimGraphNode_StateMachine_6.HitReact_Shield_Rear.AnimStateNode_1.HitReact_Shard_Rear
  nodes [Output Animation Pose(Result="Sequence Player (additive) 'SpireBoss_Hitreact_Shard_Rear'"() {bLoopAnimation=false, sequence=&/SpireBoss_Hitreact_Shard_Rear}) {
    becomeRelevantFunction {bSelfContext=true, memberGuid=#4, memberName=OnRearBecomeRelevant}
    node {becomeRelevantFunction {functionName=OnRearBecomeRelevant}}
   }]
 }
 Transition {path=AdditiveHitReacts.AnimGraphNode_StateMachine_6.HitReact_Shield_Rear.AnimStateTransitionNode_1.Transition, nodes [Result()]}
 Transition {
  path=AdditiveHitReacts.AnimGraphNode_StateMachine_6.HitReact_Shield_Rear.AnimStateTransitionNode_2.Transition
  nodes [Result(bCanEnterTransition=Get HitReact_Spire_Shard_Rear() -> (HitReact_Spire_Shard_Rear:bool))]
 }
 AnimGraph {nodes [
   n1 = Use cached pose 'Additive Layer'()
   n2 = Use cached pose 'Procedural Adjustments'() {becomeRelevantFunction {bSelfContext=true, memberGuid=#5, memberName=OnBecomeRelevant}}
   n3 = Get bEnableProceduralAdjustments() -> (bEnableProceduralAdjustments:bool)
   Output Pose(Result=BeamAttackAdditive!(InPose:Pose Link Structure=Blend Poses by bool(BlendPose_0=@n2, BlendPose_1=@n1, BlendTime_0=0.7, BlendTime_1=0.7, bActiveValue=@n3) {blendType=Linear}) -> (Pose:Pose Link Structure) {is=title AnimGraphNode_LinkedAnimLayer, layer=BeamAttackAdditive})
   Save cached pose 'Additive Layer'(Pose=AdditiveHitReacts!(InPose_0:Pose Link Structure=Use cached pose 'Base Pose'()) -> (Pose:Pose Link Structure) {is=title AnimGraphNode_LinkedAnimLayer, layer=AdditiveHitReacts}) {cacheName=Additive Layer}
   AnimGraphNode_StateMachine_0 = SpireBoss_Idle() {is=title AnimGraphNode_StateMachine, bAllowConduitEntryStates=true}
   Save cached pose 'Base Pose'(Pose=Slot 'DefaultSlot'(Source=@AnimGraphNode_StateMachine_0) {bAlwaysUpdateSourcePose=true}) {cacheName=Base Pose}
   Save cached pose 'Procedural Adjustments'(Pose=RotateMask!(InPose_1:Pose Link Structure=Use cached pose 'Additive Layer'()) -> (Pose:Pose Link Structure) {is=title AnimGraphNode_LinkedAnimLayer, layer=RotateMask}) {cacheName=Procedural Adjustments}
  ]}
 SpireBoss_Idle {
  entry=SpireBoss_Idle
  path=AnimGraph.AnimGraphNode_StateMachine_0.SpireBoss_Idle
  ids=id
  nodes [
   To Death() {is=title AnimStateAliasNode, also {aliases [SpireBoss_Idle]}}
   AnimStateNode_0 = Death() {is=title AnimStateNode}
   AnimStateNode_2 = SpireBoss_Idle() {is=title AnimStateNode}
   AnimStateNode_4 = Death Fullbody Hit React() {is=title AnimStateNode}
   AnimStateTransitionNode_0 = AnimStateTransitionNode()
   AnimStateTransitionNode_3 = AnimStateTransitionNode() {crossfadeDuration=0.1}
   AnimStateTransitionNode_7 = AnimStateTransitionNode() {bAutomaticRuleBasedOnSequencePlayerInState=true}
  ]
  links [
   {
    dst=SpireBoss_Idle
    rule=AnimGraph.AnimGraphNode_StateMachine_0.SpireBoss_Idle.AnimStateTransitionNode_0.Transition
    src=Death
    via=AnimStateTransitionNode_0
   }
   {
    dst=Death
    rule=AnimGraph.AnimGraphNode_StateMachine_0.SpireBoss_Idle.AnimStateTransitionNode_7.Transition
    src=Death Fullbody Hit React
    via=AnimStateTransitionNode_7
   }
   {
    dst=Death Fullbody Hit React
    rule=AnimGraph.AnimGraphNode_StateMachine_0.SpireBoss_Idle.AnimStateTransitionNode_3.Transition
    src=To Death
    via=AnimStateTransitionNode_3
   }
  ]
 }
 Death {
  path=AnimGraph.AnimGraphNode_StateMachine_0.SpireBoss_Idle.AnimStateNode_0.Death
  nodes [Output Animation Pose(Result=Sequence Player 'SpireBoss_Death'() {bLoopAnimation=false, sequence=&/SpireBoss_Death})]
 }
 SpireBoss_Idle {
  path=AnimGraph.AnimGraphNode_StateMachine_0.SpireBoss_Idle.AnimStateNode_2.SpireBoss_Idle
  nodes [Output Animation Pose(Result=Sequence Player 'SpireBoss_Idle'() {sequence=&/SpireBoss_Idle})]
 }
 Death Fullbody Hit React {
  path=AnimGraph.AnimGraphNode_StateMachine_0.SpireBoss_Idle.AnimStateNode_4.Death Fullbody Hit React
  nodes [
   n1 = Sequence Player() {
    becomeRelevantFunction {bSelfContext=true, memberGuid=#6, memberName=OnFullbodyHitReactBecomeRelevant}
    node {bLoopAnimation=false, becomeRelevantFunction {functionName=OnBecomeRelevant}}
   }
   Output Animation Pose(Result=@n1) {
    node {stateEntryFunction {functionName=OnStateEntry}}
    stateEntryFunction {bSelfContext=true, memberGuid=#7, memberName=OnStateEntry_FullbodyDeath}
   }
  ]
 }
 Transition {
  path=AnimGraph.AnimGraphNode_StateMachine_0.SpireBoss_Idle.AnimStateTransitionNode_0.Transition
  nodes [Result(bCanEnterTransition=NOT Boolean(A=Get Dead() -> (Dead:bool)))]
 }
 Transition {
  path=AnimGraph.AnimGraphNode_StateMachine_0.SpireBoss_Idle.AnimStateTransitionNode_3.Transition
  nodes [Result(bCanEnterTransition=Get Dead())]
 }
 Transition {path=AnimGraph.AnimGraphNode_StateMachine_0.SpireBoss_Idle.AnimStateTransitionNode_7.Transition, nodes [Result()]}
 BeamAttackAdditive {nodes [
   AnimGraphNode_BlendSpaceGraph_0 = Blendspace 'Beam AO'!(X:float=Get ShardPitchAngle() -> (ShardPitchAngle:float)) -> (Pose:Pose Link Structure) {is=title AnimGraphNode_BlendSpaceGraph}
   n1 = Input Pose()
   AnimGraphNode_StateMachine_4 = Beam Attack() {is=title AnimGraphNode_StateMachine}
   n2 = "Transform (Modify) Bone - Bone: main"!(
    ComponentPose=Local To Component(LocalPose=Apply Additive(Base=@n1, Additive=@AnimGraphNode_StateMachine_4, Alpha=0.1) {alphaBoolBlend {blendOption=Linear}})
    Alpha=1
    Translation_X:double=Get!() -> (BeamKnockbackDirection_X:double, BeamKnockbackDirection_Y:double, BeamKnockbackDirection_Z:double)
    Translation_Y:double=0
    Translation_Z:double=0
    Rotation="0,0,0"
    Scale="1,1,1"
   ) -> (Pose) {
    rotationSpace=BCS_WorldSpace
    translationMode=BMM_Additive
    translationSpace=BCS_WorldSpace
    alphaBoolBlend {blendOption=Linear}
    boneToModify {boneName=main}
   }
   n3 = "Transform (Modify) Bone - Bone: main"(ComponentPose=@n2, Alpha=1, Translation_X=0, Translation_Y=Get().BeamKnockbackDirection_Y, Translation_Z=0, Rotation="0,0,0", Scale="1,1,1") {
    rotationSpace=BCS_WorldSpace
    translationMode=BMM_Additive
    translationSpace=BCS_WorldSpace
    alphaBoolBlend {blendOption=Linear}
    boneToModify {boneName=main}
   }
   n4 = "Transform (Modify) Bone - Bone: main"!(
    ComponentPose=@n3
    Alpha=1
    Translation_X=0
    Translation_Y=0
    Translation_Z=Get().BeamKnockbackDirection_Z
    Rotation_Roll:float=0
    Rotation_Pitch:float=Get BeamAdditivePitchAlpha() -> (BeamAdditivePitchAlpha:float)
    Rotation_Yaw:float=0
    Scale="1,1,1"
   ) -> (Pose) {
    rotationMode=BMM_Additive
    rotationSpace=BCS_BoneSpace
    translationMode=BMM_Additive
    translationSpace=BCS_WorldSpace
    alphaBoolBlend {blendOption=Linear}
    boneToModify {boneName=main}
   }
   Output Pose(Result=Apply Mesh Space Additive!(Base:Pose Link Structure=Component To Local(ComponentPose=@n4), Additive:Pose Link Structure=@AnimGraphNode_BlendSpaceGraph_0, Alpha:float=1) -> (Pose:Pose Link Structure) {is=title AnimGraphNode_ApplyMeshSpaceAdditive, alphaBoolBlend {blendOption=Linear}})
  ]}
 Beam AO {path=BeamAttackAdditive.AnimGraphNode_BlendSpaceGraph_0.Beam AO, links [], nodes []}
 NewSample {
  path=BeamAttackAdditive.AnimGraphNode_BlendSpaceGraph_0.Beam AO.NewSample
  nodes [Output Animation Pose!(Result:Pose Link Structure="Sequence Player (additive) 'SpireBoss_Beam_Blades_AO_Up'"() {method=Graph, sequence=&/SpireBoss_Beam_Blades_AO_Up}) {is=title AnimGraphNode_BlendSpaceSampleResult}]
 }
 NewSample1 {
  path=BeamAttackAdditive.AnimGraphNode_BlendSpaceGraph_0.Beam AO.NewSample1
  nodes [Output Animation Pose(Result="Sequence Player (additive) 'SpireBoss_Beam_Blades_AO_Down'"() {method=Graph, sequence=&/SpireBoss_Beam_Blades_AO_Down})]
 }
 Beam Attack {
  entry=Empty
  path=BeamAttackAdditive.AnimGraphNode_StateMachine_4.Beam Attack
  ids=id
  nodes [
   AnimStateNode_0 = Empty() {is=title AnimStateNode}
   AnimStateNode_1 = Beam Fire() {is=title AnimStateNode}
   AnimStateTransitionNode_0 = AnimStateTransitionNode() {crossfadeDuration=0}
   AnimStateTransitionNode_1 = AnimStateTransitionNode() {bAutomaticRuleBasedOnSequencePlayerInState=true}
  ]
  links [
   {
    dst=Empty
    rule=BeamAttackAdditive.AnimGraphNode_StateMachine_4.Beam Attack.AnimStateTransitionNode_1.Transition
    src=Beam Fire
    via=AnimStateTransitionNode_1
   }
   {
    dst=Beam Fire
    rule=BeamAttackAdditive.AnimGraphNode_StateMachine_4.Beam Attack.AnimStateTransitionNode_0.Transition
    src=Empty
    via=AnimStateTransitionNode_0
   }
  ]
 }
 Empty {
  path=BeamAttackAdditive.AnimGraphNode_StateMachine_4.Beam Attack.AnimStateNode_0.Empty
  nodes [Output Animation Pose(Result=Additive Identity Pose()) {is=title AnimGraphNode_StateResult}]
 }
 Beam Fire {
  path=BeamAttackAdditive.AnimGraphNode_StateMachine_4.Beam Attack.AnimStateNode_1.Beam Fire
  nodes [Output Animation Pose(Result="Sequence Player (additive) 'SpireBoss_Beam_Fire_Additive'"() {bLoopAnimation=false, sequence=&/SpireBoss_Beam_Fire_Additive}) {
    becomeRelevantFunction {bSelfContext=true, memberGuid=#8, memberName=OnBeamFireAdditive}
    node {becomeRelevantFunction {functionName=OnBecomeRelevant_0}}
   }]
 }
 Transition {
  path=BeamAttackAdditive.AnimGraphNode_StateMachine_4.Beam Attack.AnimStateTransitionNode_0.Transition
  nodes [Result(bCanEnterTransition=Get bApplyBeamAdditive() -> (bApplyBeamAdditive:bool))]
 }
 Transition {path=BeamAttackAdditive.AnimGraphNode_StateMachine_4.Beam Attack.AnimStateTransitionNode_1.Transition, nodes [Result()]}
 EventGraph {links [], nodes []}
 HitReactGraph {nodes [
   HitReact_Shield_Right!() -> (OutputDelegate, then, HitIntensity:double=~) {is=title K2Node_CustomEvent}
   > Set HitReact_Spire_Shard_Right(HitReact_Spire_Shard_Right:bool=true) -> (Output_Get:bool)
   HitReact_Shield_Rear!() -> (OutputDelegate, then, HitIntensity:double=~) {is=title K2Node_CustomEvent}
   > Set HitReact_Spire_Shard_Rear(HitReact_Spire_Shard_Rear:bool=true) -> (Output_Get:bool)
   HitReact_Main!() -> (OutputDelegate, then, HitIntensity:double=~) {is=title K2Node_CustomEvent}
   > Set HitReact_Spire_Main(HitReact_Spire_Main:bool=true) -> (Output_Get:bool)
   n1 = SetIsDead!() -> (OutputDelegate, then, bIsDead:bool=~) {is=title K2Node_CustomEvent}
   > Set Dead(Dead:bool=@n1.bIsDead) -> (Output_Get:bool)
   HitReact_Shield_Left!() -> (OutputDelegate, then, HitIntensity:double=~) {is=title K2Node_CustomEvent}
   > Set HitReact_Spire_Shard_Left(HitReact_Spire_Shard_Left:bool=true) -> (Output_Get:bool)
  ]}
 OnBeamFireAdditive {nodes [
   OnBeamFireAdditive!() -> (then, Context:"Anim Update Context Structure (by ref)", Node:"Anim Node Reference Structure (by ref)") {is=title K2Node_FunctionEntry}
   > Set bHasPlayedAdditive(bHasPlayedAdditive:bool=true) -> (Output_Get:bool)
   > Return Node()
  ]}
 OnBecomeRelevant {nodes [
   OnBecomeRelevant!() -> (then, Context:"Anim Update Context Structure (by ref)", Node:"Anim Node Reference Structure (by ref)") {is=title K2Node_FunctionEntry}
   > Set MaskLookDirection(MaskLookDirection:double=Get WindupBeamYaw() -> (WindupBeamYaw:float)) -> (Output_Get:double)
   > Return Node()
  ]}
 OnFullbodyHitReactBecomeRelevant {nodes [
   n1 = OnFullbodyHitReactBecomeRelevant!() -> (then, Context:"Anim Update Context Structure (by ref)", Node:"Anim Node Reference Structure (by ref)") {is=title K2Node_FunctionEntry}
   > SetSequenceWithInertialBlending!(
    execute
    UpdateContext:"Anim Update Context Structure (by ref)"=@n1
    SequencePlayer:"Sequence Player Reference Structure (by ref)"=Convert to Sequence Player!(Node:"Anim Node Reference Structure (by ref)"=@n1.Node) -> (SequencePlayer:Sequence Player Reference Structure, Result:bool)
    Sequence:Anim Sequence Base Object Reference=Get HitSequence() -> (HitSequence:Anim Sequence Object Reference)
    BlendTime:float=0.2
   ) -> (then, ReturnValue:Sequence Player Reference Structure)
   > Return Node()
  ]}
 OnLeftHitBecomeRelevant {nodes [
   OnLeftHitBecomeRelevant!() -> (then, Context:"Anim Update Context Structure (by ref)", Node:"Anim Node Reference Structure (by ref)") {is=title K2Node_FunctionEntry}
   > Set HitReact_Spire_Shard_Left(HitReact_Spire_Shard_Left=false)
   > Return Node()
  ]}
 OnMainBecomeRelevant {nodes [
   OnMainBecomeRelevant!() -> (then, Context:"Anim Update Context Structure (by ref)", Node:"Anim Node Reference Structure (by ref)") {is=title K2Node_FunctionEntry}
   > Set HitReact_Spire_Main(HitReact_Spire_Main=false)
   > Return Node()
  ]}
 OnRearBecomeRelevant {nodes [
   OnRearBecomeRelevant!() -> (then, Context:"Anim Update Context Structure (by ref)", Node:"Anim Node Reference Structure (by ref)") {is=title K2Node_FunctionEntry}
   > Set HitReact_Spire_Shard_Rear(HitReact_Spire_Shard_Rear=false)
   > Return Node()
  ]}
 OnRightHitBecomeRelevant {nodes [
   OnRightHitBecomeRelevant!() -> (then, Context:"Anim Update Context Structure (by ref)", Node:"Anim Node Reference Structure (by ref)") {is=title K2Node_FunctionEntry}
   > Set HitReact_Spire_Shard_Right(HitReact_Spire_Shard_Right=false)
   > Return Node()
  ]}
 OnStateEntry_FullbodyDeath {nodes [
   OnStateEntry_FullbodyDeath!() -> (then, Context:"Anim Update Context Structure (by ref)", Node:"Anim Node Reference Structure (by ref)") {is=title K2Node_FunctionEntry}
   > n1 = Branch(Condition=Get bAppendageSlamActive() -> (bAppendageSlamActive:bool))
   > SetHitAppendage()
   > Set HitSequence(HitSequence:Anim Sequence Object Reference=Find(
     TargetMap:Map of SpireAppendageKind Enums to Anim Sequence Object References=Get HitReactSequence() -> (HitReactSequence:Map of SpireAppendageKind Enums to Anim Sequence Object References=NewEnumerator0)
     Key:"SpireAppendageKind Enum (by ref)"=Get SlamAppendage() -> (SlamAppendage:SpireAppendageKind Enum=NewEnumerator0)
    ) -> (Value:Anim Sequence Object Reference)) -> (Output_Get:Anim Sequence Object Reference)
   > Return Node()
   n2 = Branch(execute=@n1.else, Condition=Get bBeamSweepActive() -> (bBeamSweepActive:bool))
   > Set HitSequence(HitSequence=&/SpireBoss_Hitreact_FullBody_Beam)
   Set HitSequence(execute=@n2.else, HitSequence=&/SpireBoss_Hitreact_FullBody_Idle)
  ]}
 RotateMask {nodes [
   n1 = "Transform (Modify) Bone - Bone: shard_center"!(
    ComponentPose=Local To Component(LocalPose=Use cached pose 'IncomingPose'())
    Alpha=1
    Translation="0,0,0"
    Rotation_Roll:float=0
    Rotation_Pitch:float=0
    Rotation_Yaw:float=Get MaskLookDirection() -> (MaskLookDirection:double)
    Scale="1,1,1"
   ) -> (Pose) {
    rotationMode=BMM_Replace
    rotationSpace=BCS_WorldSpace
    alphaBoolBlend {blendOption=Linear}
    alphaScaleBiasClamp {interpSpeedDecreasing=30}
    boneToModify {boneName=shard_center}
   }
   n2 = "Transform (Modify) Bone - Bone: shard_small_master"!(
    ComponentPose=@n1
    Alpha=1
    Translation="0,0,0"
    Rotation_Roll:float=0
    Rotation_Pitch:float=0
    Rotation_Yaw:float=Get MaskLookDirection()
    Scale="1,1,1"
   ) -> (Pose) {
    rotationMode=BMM_Replace
    rotationSpace=BCS_WorldSpace
    alphaBoolBlend {blendOption=Linear}
    alphaScaleBiasClamp {interpSpeedDecreasing=30}
    boneToModify {boneName=shard_small_master}
   }
   Output Pose(Result=Component To Local(ComponentPose=@n2))
   Save cached pose 'IncomingPose'(Pose=Input Pose() {name=InPose_1}) {cacheName=IncomingPose}
  ]}
 SetHitAppendage {nodes [
   SetHitAppendage() {is=title K2Node_FunctionEntry}
   > n1 = Branch(Condition=Get bAppendageFrontLeft() -> (bAppendageFrontLeft:bool))
   > Set SlamAppendage(SlamAppendage:SpireAppendageKind Enum=NewEnumerator0) -> (Output_Get:SpireAppendageKind Enum=NewEnumerator0)
   n2 = Branch(execute=@n1.else, Condition=Get bAppendageFrontRight() -> (bAppendageFrontRight:bool))
   > Set SlamAppendage(SlamAppendage=NewEnumerator1) -> (Output_Get=NewEnumerator0)
   n3 = Branch(execute=@n2.else, Condition=Get bAppendageBack() -> (bAppendageBack:bool))
   > Set SlamAppendage(SlamAppendage=NewEnumerator2) -> (Output_Get=NewEnumerator0)
   Set SlamAppendage(execute=@n3.else, SlamAppendage=NewEnumerator0) -> (Output_Get=NewEnumerator0)
  ]}
]
variables [
 Dead {type=boolean}
 HitReact_Spire_Main {type=boolean}
 HitReact_Spire_Shard_Left {type=boolean}
 HitReact_Spire_Shard_Rear {type=boolean}
 HitReact_Spire_Shard_Right {type=boolean}
 HitReactSequence {default {
   Back=&/SpireBoss_Hitreact_FullBody_Slam_Rear
   Front Left=&/SpireBoss_Hitreact_FullBody_Slam_Left
   Front Right=&/SpireBoss_Hitreact_FullBody_Slam_Right
  }}
 HitSequence {type=AnimSequence, default=null}
 Mask Look Direction {type=number}
 SlamAppendage {type=SpireAppendageKind, default=Front Left}
]
